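# Niagara effect for the SpicyPumpkin emote: an autumn leaf tornado with swirling wind and dust.
# GPU vortex emitters spin leaves at bottom and top, while trailing leaves tumble with drag and collision, and the owner's velocity adjusts spawning and fading.
# Layered wind meshes rise, rotate and grow, with one set following a gadget pumpkin bone.
# Dust sprites animate via SubUV; wind and dust colors shift with time of day.
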NiagaraSystem /BRCosmetics/Animation/Game/MainPlayer/Emotes/SpicyPumpkin/FX/NS_Emote_SpicyPumpkin_Leaves
dependencies {
 usedDataInterfaces [
  &NiagaraDataInterfaceCurve
  &NiagaraDataInterfaceCollisionQuery
  &NiagaraDataInterfaceLandscape
  &NiagaraDataInterfaceVectorCurve
  &NiagaraDataInterfaceArrayFloat3
  &NiagaraDataInterfaceSkeletalMesh
 ]
 usedDynamicInputs [&MakeVector]
 usedModules [&SystemState]
}
emitters {
 Dark_Wind {
  emitterUpdateScript [
   &EmitterState(Life Cycle Mode=::NewEnumerator1, Loop Duration=10, Scalability Mode=::NewEnumerator1, UseLoopDelay=false)
   &SpawnRate(SpawnRate=3, Use Spawn Probability=false)
  ]
  particleSpawnScript [
   &InitializeParticle(
    Color=&RandomRangeLinearColor(Minimum=<rgba 0.2708 0.0856 0.0254 1>, Maximum=<rgba 0.1198 0.036 0.0082 1>)
    Lifetime=1.5
    Mesh Scale=&Multiply_Vector(
     A=&VectorFromFloat(Value=2)
     B=&RandomRangeVector(Minimum=<0.2 0.02 0.15>, Maximum=<0.15 0.02 0.15>)
    )
    Mesh Scale Mode=::NewEnumerator4
    Position=<0 0 -50>
    Position Mode=::NewEnumerator1
    UsePositionOffset=false
    Write Material Random=true
   )
   &CameraOffset(Camera Offset Amount=15)
   &InitialMeshOrientation(
    Mesh Axis To Orient=<0 1 0>
    Mesh Orientation Mode=::NewEnumerator2
    Orientation Coordinate Space=::World
    Orientation Vector=<0 0 1>
    Rotation=&RandomRangeVector(Minimum=<0 0 0>, Maximum=<0 1 0>)
    Use Rotation Vector=true
   )
   &AddVelocity(Velocity=<0 0 107>)
  ]
  particleUpdateScript [
   &ScaleColor(
    Scale Alpha=&FNTimeOfDayScalar(Night Scale=1, Morning Scale=1, Evening Scale=1)
    Scale RGB=&Multiply_VectorByFloat(Vector=<1 1 1>, Float=&FNTimeOfDayScalar(Night Scale=1.5, Morning Scale=1, Evening Scale=1))
    ScaleA=true
    ScaleRGB=true
   ) {module=ScaleColor002}
   &ParticleState()
   &UpdateMeshOrientation(Rotation Coordinate Space=::NewEnumerator0, Rotation Rate=0.2, Rotation Vector=<0 0 -1>)
   &ScaleMeshSize(Scale Factor=&VectorFromFloat(Value=&FloatFromCurve(FloatCurve=DataInterface() {propertyValues {Curve {keys [<at 0 0>,<at 0.45 0.4>,<at 1 1.6>]}}}, Scale Curve=1.25)))
   &DynamicMaterialParameters(
    Index 0 Param 1=&FloatFromCurve(FloatCurve=DataInterface() {propertyValues {Curve {keys [
         {arriveTangent=2.4535, interpMode=RCIM_Cubic, leaveTangent=2.4535, tangentMode=RCTM_User, time=0.017, value=0}
         {interpMode=RCIM_Cubic, tangentMode=RCTM_User, time=0.5, value=0.5}
         {arriveTangent=-2.5865, tangentMode=RCTM_User, time=1, value=0}
        ]}}})
    Module.Index0Param0Write=true
    Module.Index0Param1Write=false
    Module.Index0Param2Write=false
    Module.Index0Param3Write=false
   )
   &InheritSourceMovement(ApplyForce=false, ApplyPosition=true, ApplyPositionFalloff=false) {enabled=false}
   &Drag(
    Drag=&FloatFromCurve(
     FloatCurve=DataInterface() {propertyValues {Curve {defaultValue=1, keys [{tangentMode=RCTM_User, time=0, value=0}, {tangentMode=RCTM_User, time=1, value=1}]}}}
     Scale Curve=2
    )
    Use Linear Drag=true
    Use Rotational Drag=true
   )
   &SolveForcesAndVelocity(Clamp Velocity=false, Limit Acceleration=false)
   &ScaleColor(
    Scale Alpha=&FloatFromCurve(
     FloatCurve=DataInterface() {propertyValues {Curve {keys [
         {arriveTangent=3.7535, interpMode=RCIM_Cubic, leaveTangent=3.7535, tangentMode=RCTM_User, time=0, value=0}
         {interpMode=RCIM_Cubic, tangentMode=RCTM_User, time=0.5, value=1}
         {tangentMode=RCTM_User, time=1, value=0}
        ]}}}
     Scale Curve=0.3
    )
    Scale RGB=&MakeVectorFromLinearColorRGB(Color=&FNTimeOfDayColor(
      Night Color=<rgba 0.1253 0.1914 0.5 1>
      Morning Color=<rgba 0.5 0.3383 0.3719 1>
      Day Color=<rgba 0.5 0.5 0.5 1>
      Evening Color=<rgba 0.4474 0.4547 0.4578 1>
     ))
    ScaleA=true
    ScaleRGB=true
   )
  ]
  rendererClasses [&NiagaraMeshRendererProperties]
  renderers [{rendererClass=&NiagaraMeshRendererProperties, rendererIndex=0}]
 }
 Leaf_Tornado_Bottom {
  simTarget=GPUComputeSim
  emitterUpdateScript [
   &EmitterState(Life Cycle Mode=::NewEnumerator1, UseLoopDelay=false)
   &SpawnRate(
    SpawnRate=&MakeCustomFloatFromBool(
     Bool=&SetBoolByFloatComparison(A=&VectorLength(Vector=$User.OwnerLocalVelocity), B=1)
     True Float=40
     False Float=30
    )
    Use Spawn Probability=false
   )
  ]
  particleSpawnScript [
   &InitializeParticle(
    Color Maximum=<rgba 0.7906 0.3939 0 1>
    Color Minimum=<rgba 0.6573 0.1236 0 1>
    Color Mode=::NewEnumerator2
    Lifetime Min=1.5
    Lifetime Mode=::NewEnumerator1
    Mass Max=1.5
    Mass Mode=::NewEnumerator3
    Sprite Rotation Mode=::NewEnumerator2
    Sprite Size Mode=::NewEnumerator1
    Uniform Sprite Size Max=8
    Uniform Sprite Size Min=4
    UsePositionOffset=true
    Write Material Random=true
   )
   &ShapeLocation(
    Cylinder Height=20
    Cylinder Radius=25
    Enable Lathe Profile=true
    Hemicircle Internal Cap=false
    Hemicircle Y=true
    Lathe Profile=DataInterface() {propertyValues {Curve {keys [
        {arriveTangent=-0.8754, interpMode=RCIM_Cubic, leaveTangent=-0.8754, tangentMode=RCTM_User, time=0, value=3}
        {
         arriveTangent=-2.3448
         arriveTangentWeight=0.1283
         interpMode=RCIM_Cubic
         leaveTangent=-0.5213
         leaveTangentWeight=0.188
         tangentMode=RCTM_Break
         tangentWeightMode=RCTWM_WeightedBoth
         time=0.5
         value=1.5
        }
        {arriveTangent=-0.9585, interpMode=RCIM_Cubic, leaveTangent=-0.9585, tangentMode=RCTM_User, time=1, value=0.5}
       ]}}}
    MaskSpawnGroup=false
    Non Uniform Scale=<1 1 -3>
    Offset=<0 0 13>
    Offset Mode=::NewEnumerator0
    Override Seed=false
    Shape Primitive=::NewEnumerator1
    Surface Only=true
    UseOwnerScale=true
   )
   &InitialMeshOrientation(Use Rotation Vector=false)
   &ShapeLocation(
    Cylinder Height=40
    Cylinder Radius=10
    MaskSpawnGroup=false
    Offset=<0 0 -10>
    Offset Mode=::NewEnumerator0
    Override Seed=false
    Shape Primitive=::NewEnumerator1
    Surface Only=true
    Surface Only Band Thickness=1
    UseOwnerScale=true
   ) {enabled=false, module=ShapeLocation001}
  ]
  particleUpdateScript [
   &ParticleState()
   &VortexForce(
    Origin Pull Amount=800
    UseOriginOffset=true
    Vortex Force Amount=&Multiply_Float(A=&RandomRangeFloat(Minimum=-5000, Maximum=-7500), B=1)
    Vortex Origin Offset:Vector3f=<0 0 3>
   )
   /SetVariables(
    Particles.Velocity=<0 0 10>
    Transient.PhysicsForce:Vector3f=&Multiply_Vector(A=$Transient.PhysicsForce, B=<4 4 0>)
   ) {isSetParameters=true, module=SetVariables_#1}
   &InheritSourceMovement(ApplyForce=false, ApplyPosition=true, ApplyPositionFalloff=false) {enabled=false}
   &SolveForcesAndVelocity(
    Clamp Velocity=false
    DeltaTime:NiagaraFloat=/Niagara/DynamicInputs/Math/Min_Float(A:NiagaraFloat=$Engine.DeltaTime, B:NiagaraFloat=$User.DeltaTimeMax)
    Limit Acceleration=false
   )
   &ScaleSpriteSize(Uniform Curve Sprite Scale=DataInterface() {propertyValues {Curve {keys [
        {arriveTangent=14.5835, interpMode=RCIM_Cubic, leaveTangent=14.5835, tangentMode=RCTM_User, time=0, value=0}
        {arriveTangent=1.1111, interpMode=RCIM_Cubic, leaveTangent=1.1111, time=0.1, value=1}
        {arriveTangent=-1.1111, interpMode=RCIM_Cubic, leaveTangent=-1.1111, time=0.9, value=1}
        {arriveTangent=-8.7113, interpMode=RCIM_Cubic, leaveTangent=-8.7113, tangentMode=RCTM_User, time=1, value=0}
       ]}}})
   &AlignSpriteToMeshOrientation(SetAlignment=true, SetFacing=true)
   /Niagara/Modules/Collision/AlignParticlesWithCollisionPlane(Allow Alignment on Y=false, Allow Alignment on Z=false, Rotational Dampening Power=1) {enabled=false}
  ]
  rendererClasses [&NiagaraSpriteRendererProperties, &NiagaraMeshRendererProperties]
  renderers [
   {rendererClass=&NiagaraSpriteRendererProperties}
   {rendererClass=&NiagaraMeshRendererProperties, rendererIndex=1}
  ]
 }
 Leaf_Tornado_Top {
  simTarget=GPUComputeSim
  emitterUpdateScript [
   &EmitterState(Life Cycle Mode=::NewEnumerator1, UseLoopDelay=false)
   &SpawnRate(SpawnRate=30, Use Spawn Probability=false)
  ]
  particleSpawnScript [
   &InitializeParticle(
    Color Maximum=<rgba 0.7906 0.3939 0 1>
    Color Minimum=<rgba 0.6573 0.1236 0 1>
    Color Mode=::NewEnumerator2
    Lifetime Max=1.2
    Lifetime Mode=::NewEnumerator1
    Mass Max=1.5
    Mass Mode=::NewEnumerator3
    Sprite Rotation Mode=::NewEnumerator2
    Sprite Size Mode=::NewEnumerator1
    Uniform Sprite Size Max=8
    Uniform Sprite Size Min=4
    UsePositionOffset=true
    Write Material Random=true
   )
   &ShapeLocation(
    MaskSpawnGroup=false
    Offset=<0 0 25>
    Offset Mode=::NewEnumerator0
    Override Seed=false
    Ring Radius=50
    Shape Primitive=::NewEnumerator4
    UseOwnerScale=true
   )
   &InitialMeshOrientation(Use Rotation Vector=false)
   &ShapeLocation(
    Cylinder Height=40
    Cylinder Radius=10
    MaskSpawnGroup=false
    Offset=<0 0 -10>
    Offset Mode=::NewEnumerator0
    Override Seed=false
    Shape Primitive=::NewEnumerator1
    Surface Only=true
    Surface Only Band Thickness=1
    UseOwnerScale=true
   ) {enabled=false, module=ShapeLocation001}
  ]
  particleUpdateScript [
   &ParticleState()
   &VortexForce(
    Origin Pull Amount=400
    UseOriginOffset=false
    Vortex Force Amount=&Multiply_Float(A=&RandomRangeFloat(Minimum=-5000, Maximum=-7500), B=1)
   )
   /SetVariables(
    Particles.Velocity=<0 0 50>
    Transient.PhysicsForce:Vector3f=&Multiply_Vector(A=$Transient.PhysicsForce, B=<7 7 1>)
   ) {isSetParameters=true, module=SetVariables_#1}
   &InheritSourceMovement(ApplyForce=false, ApplyPosition=true, ApplyPositionFalloff=false) {enabled=false}
   &PointForce(Force Strength=1700, Module.Use Falloff=false, Module.UseLinearForce=false)
   &SolveForcesAndVelocity(
    Clamp Velocity=false
    DeltaTime:NiagaraFloat=&Min_Float(A:NiagaraFloat=$Engine.DeltaTime, B:NiagaraFloat=$User.DeltaTimeMax)
    Limit Acceleration=false
   )
   &ScaleSpriteSize(Uniform Curve Sprite Scale=DataInterface() {propertyValues {Curve {keys [
        {arriveTangent=14.5835, interpMode=RCIM_Cubic, leaveTangent=14.5835, tangentMode=RCTM_User, time=0, value=0}
        {arriveTangent=1.1111, interpMode=RCIM_Cubic, leaveTangent=1.1111, time=0.1, value=1}
        {arriveTangent=-1.1111, interpMode=RCIM_Cubic, leaveTangent=-1.1111, time=0.9, value=1}
        {arriveTangent=-8.7113, interpMode=RCIM_Cubic, leaveTangent=-8.7113, tangentMode=RCTM_User, time=1, value=0}
       ]}}})
   &AlignSpriteToMeshOrientation(SetAlignment=true, SetFacing=true)
   &AlignParticlesWithCollisionPlane(Allow Alignment on Y=false, Allow Alignment on Z=false, Rotational Dampening Power=1) {enabled=false}
  ]
  rendererClasses [&NiagaraSpriteRendererProperties, &NiagaraMeshRendererProperties]
  renderers [
   {rendererClass=&NiagaraSpriteRendererProperties}
   {rendererClass=&NiagaraMeshRendererProperties, rendererIndex=1}
  ]
 }
 Light_Wind {
  emitterUpdateScript [
   &EmitterState(Life Cycle Mode=::NewEnumerator1, Loop Duration=10, Scalability Mode=::NewEnumerator1, UseLoopDelay=false)
   &SpawnRate(SpawnRate=10, Use Spawn Probability=false)
  ]
  particleSpawnScript [
   &InitializeParticle(
    Color=&RandomRangeLinearColor(Minimum=<rgba 0.875 0.4101 0.1887 0.3>, Maximum=<rgba 0.1198 0.036 0.0082 0.5>)
    Lifetime=1.5
    Mesh Scale=&Multiply_Vector(
     A=&VectorFromFloat(Value=2)
     B=&RandomRangeVector(Minimum=<0.2 0.02 0.15>, Maximum=<0.15 0.02 0.15>)
    )
    Mesh Scale Mode=::NewEnumerator4
    Position=<0 0 -50>
    Position Mode=::NewEnumerator1
    UsePositionOffset=false
    Write Material Random=true
   )
   &CameraOffset(Camera Offset Amount=15)
   &InitialMeshOrientation(
    Mesh Axis To Orient=<0 1 0>
    Mesh Orientation Mode=::NewEnumerator2
    Orientation Coordinate Space=::World
    Orientation Vector=<0 0 1>
    Rotation=&RandomRangeVector(Minimum=<0 0 0>, Maximum=<0 1 0>)
    Use Rotation Vector=true
   )
   &AddVelocity(Velocity=<0 0 107>)
  ]
  particleUpdateScript [
   &ScaleColor(
    Scale Alpha=&FNTimeOfDayScalar(Night Scale=1, Morning Scale=1, Evening Scale=1)
    Scale RGB=&Multiply_VectorByFloat(Vector=<1 1 1>, Float=&FNTimeOfDayScalar(Night Scale=1.5, Morning Scale=1, Evening Scale=1))
    ScaleA=true
    ScaleRGB=true
   ) {module=ScaleColor001}
   &ParticleState()
   &UpdateMeshOrientation(Rotation Coordinate Space=::NewEnumerator0, Rotation Rate=0.2, Rotation Vector=<0 0 -1>)
   &ScaleMeshSize(Scale Factor=&VectorFromFloat(Value=&FloatFromCurve(FloatCurve=DataInterface() {propertyValues {Curve {keys [<at 0 0>,<at 0.45 0.4>,<at 1 1.6>]}}}, Scale Curve=1.25)))
   &DynamicMaterialParameters(
    Index 0 Param 1=&FloatFromCurve(FloatCurve=DataInterface() {propertyValues {Curve {keys [
         {arriveTangent=2.4535, interpMode=RCIM_Cubic, leaveTangent=2.4535, tangentMode=RCTM_User, time=0.017, value=0}
         {interpMode=RCIM_Cubic, tangentMode=RCTM_User, time=0.5, value=0.5}
         {arriveTangent=-2.5865, tangentMode=RCTM_User, time=1, value=0}
        ]}}})
    Module.Index0Param0Write=true
    Module.Index0Param1Write=false
    Module.Index0Param2Write=false
    Module.Index0Param3Write=false
   )
   &InheritSourceMovement(ApplyForce=false, ApplyPosition=true, ApplyPositionFalloff=false) {enabled=false}
   &Drag(
    Drag=&FloatFromCurve(
     FloatCurve=DataInterface() {propertyValues {Curve {defaultValue=1, keys [{tangentMode=RCTM_User, time=0, value=0}, {tangentMode=RCTM_User, time=1, value=1}]}}}
     Scale Curve=2
    )
    Use Linear Drag=true
    Use Rotational Drag=true
   )
   &SolveForcesAndVelocity(Clamp Velocity=false, Limit Acceleration=false)
   &ScaleColor(
    Scale Alpha=&FloatFromCurve(
     FloatCurve=DataInterface() {propertyValues {Curve {keys [
         {arriveTangent=3.7535, interpMode=RCIM_Cubic, leaveTangent=3.7535, tangentMode=RCTM_User, time=0, value=0}
         {interpMode=RCIM_Cubic, tangentMode=RCTM_User, time=0.5, value=1}
         {tangentMode=RCTM_User, time=1, value=0}
        ]}}}
     Scale Curve=0.3
    )
    Scale RGB=&MakeVectorFromLinearColorRGB(Color=&FNTimeOfDayColor(
      Night Color=<rgba 0.1253 0.1914 0.5 1>
      Morning Color=<rgba 0.5 0.3383 0.3719 1>
      Day Color=<rgba 0.5 0.5 0.5 1>
      Evening Color=<rgba 0.4474 0.4547 0.4578 1>
     ))
    ScaleA=true
    ScaleRGB=true
   )
  ]
  rendererClasses [&NiagaraMeshRendererProperties]
  renderers [{rendererClass=&NiagaraMeshRendererProperties, rendererIndex=0}]
 }
 Local_Dust {
  emitterUpdateScript [
   &EmitterState(Life Cycle Mode=::NewEnumerator1, Loop Duration=1, UseLoopDelay=true)
   &SpawnRate(SpawnRate=30, Use Spawn Probability=false)
  ]
  particleSpawnScript [
   &InitializeParticle(
    Color=&FNTimeOfDayColor(
     Night Color=<rgba 0.4132 0.4522 0.5833 0.2>
     Morning Color=<rgba 0.2708 0.1609 0.1217 0.2>
     Day Color=<rgba 0.2708 0.1609 0.1217 0.2>
     Evening Color=<rgba 0.2708 0.1609 0.1217 0.2>
    )
    Lifetime Max=1
    Lifetime Min=0.5
    Lifetime Mode=::NewEnumerator1
    Sprite Rotation Angle=&RandomRangeFloat(Minimum=-0.05, Maximum=0.3)
    Sprite Rotation Mode=::NewEnumerator1
    Sprite Size Mode=::NewEnumerator1
    Uniform Sprite Size Max=50
    Uniform Sprite Size Min=60
    UsePositionOffset=true
    Write Material Random=true
   )
   &ShapeLocation(
    Cylinder Height=15
    Cylinder Radius=20
    MaskSpawnGroup=false
    Override Seed=false
    Shape Primitive=::NewEnumerator1
    Surface Only=false
    UseOwnerScale=true
   )
   &AddVelocity(Velocity=&RandomRangeVector(Minimum=<-2 -2 6>, Maximum=<2 2 10>, Evaluation Type=::NewEnumerator1)) {enabled=false}
   &InheritVelocity(Module.Limit Speed=true, Velocity Limit=5, Velocity Scale=<0.25 0.25 0.25>) {enabled=false}
   &PointForce(Force Strength=50, Module.Use Falloff=false, Module.UseLinearForce=false)
   &ApplyInitialForces()
   /SetVariables(Particles.LayerColor:LinearColor=<rgba 1 0.6847 0.3958 0.5>, Particles.LayerScale:Vector2f=<40 40>) {isSetParameters=true}
  ]
  particleUpdateScript [
   &ParticleState()
   &SubUVAnimation(
    Sub UV Play Rate:NiagaraFloat=&RandomRangeFloat(Minimum=0.03, Maximum=0.05)
    SubUV Animation Mode:ENiagara_SubUVLookupMode=::NewEnumerator3
    SubUV Start Time Offset:NiagaraFloat=&RandomRangeFloat(Minimum=0, Maximum=63)
   )
   &DynamicMaterialParameters(
    Index 0 Param 1=&FloatFromCurve(FloatCurve=DataInterface() {propertyValues {Curve {keys [<at 0 -0.25>,<at 1 -1>]}}})
    Index 0 Param 4=&OneMinusFloat(Float=$Particles.NormalizedAge)
    Module.Param0WriteEnabled=true
    Module.Param1WriteEnabled=false
    Module.Param2WriteEnabled=false
    Module.Param3WriteEnabled=true
   )
   &ScaleSpriteSize(Scale Factor=&Vector2DFromFloat(Value=&FloatFromCurve(
      FloatCurve=DataInterface() {propertyValues {Curve {keys [
          {arriveTangent=1.8751, interpMode=RCIM_Cubic, leaveTangent=1.8751, tangentMode=RCTM_User, time=0, value=0.5}
          {arriveTangent=0.72, interpMode=RCIM_Cubic, leaveTangent=0.72, tangentMode=RCTM_User, time=1, value=1.25}
         ]}}}
      Scale Curve=1.5
     )))
   &Drag(Drag=0.5, Use Linear Drag=true, Use Rotational Drag=false)
   &SolveForcesAndVelocity(Clamp Velocity=false, Limit Acceleration=false)
   &ScaleColor(
    Scale Alpha=&FloatFromCurve(FloatCurve=DataInterface() {propertyValues {Curve {keys [<at 0 1>,<at 0.3 1>,<at 0.85 1>,<at 1 0>]}}})
    ScaleA=true
    ScaleRGB=true
   )
   &ScaleColor(ScaleA=true, ScaleRGB=true) {module=ScaleColor001}
  ]
  rendererClasses [&NiagaraSpriteRendererProperties]
  renderers [
   {rendererClass=&NiagaraSpriteRendererProperties}
   {rendererClass=&NiagaraSpriteRendererProperties, rendererIndex=1}
  ]
 }
 Skeletal_Wind {
  emitterUpdateScript [
   &EmitterState(
    Delay First Loop Only=true
    Life Cycle Mode=::NewEnumerator1
    Loop Delay=0.1
    Loop Duration=10
    Scalability Mode=::NewEnumerator1
    UseLoopDelay=true
   )
   &SpawnRate(SpawnRate=8, Use Spawn Probability=false)
  ]
  particleSpawnScript [
   &InitializeParticle(
    Color=&RandomRangeLinearColor(Minimum=<rgba 0.4792 0.2019 0.1198 0.5>, Maximum=<rgba 0.5365 0.1923 0.0782 0.3>)
    Lifetime=1.5
    Mesh Scale=&Multiply_Vector(
     A=&VectorFromFloat(Value=2)
     B=&RandomRangeVector(Minimum=<0.2 0.05 0.15>, Maximum=<0.15 0.05 0.15>)
    )
    Mesh Scale Mode=::NewEnumerator4
    Position=<0 0 100>
    Position Mode=::NewEnumerator1
    UsePositionOffset=false
    Write Material Random=true
   )
   &CameraOffset(Camera Offset Amount=15)
   &InitialMeshOrientation(
    Mesh Axis To Orient=<0 1 0>
    Mesh Orientation Mode=::NewEnumerator2
    Orientation Coordinate Space=::World
    Orientation Vector=<0 0 1>
    Rotation=&RandomRangeVector(Minimum=<0 0 0>, Maximum=<0 1 0>)
    Use Rotation Vector=true
   )
   &AddVelocity()
   &ShapeLocation(MaskSpawnGroup=false, Override Seed=false, UseOwnerScale=true)
   &SkeletalMeshLocation(
    Bone Sampling Mode=::NewEnumerator4
    MaskSpawnGroup=false
    Orientation Sampling=::NewEnumerator1
    Position Sampling=::NewEnumerator2
    Sampled Position Offset=<0 0 -20>
    Sampled Position Offset Coordinate Space:ENiagaraSkelMeshTransforms=::NewEnumerator0
    Skeletal Mesh=DataInterface() {propertyValues {
      DefaultMesh=/BRCosmetics/Animation/Props/Gadgets_SpicyPumpkin/Meshes/SpicyPumpkin_Gadget
      ExcludeBoneName=root
      OptionalComponentClass=null
      PreviewMesh=&SpicyPumpkin_Gadget
      SkinningMode=SkinOnTheFly
      SoftSourceActor=null
      SourceMode=Default
      UvSetIndex=0
      WholeMeshLOD=-1
      bExcludeBone=false
      bReadDeformedGeometry=true
      bRequireCurrentFrameData=true
      ComponentTags []
      FilteredBones [SpicyPumpkin_Pumpkin_1_jnt]
      FilteredSockets []
      MeshUserParameter {parameter {name=None, typeDefHandle {registeredTypeIndex=1}, varData []}}
      SamplingRegions []
     }}
   ) {module=SkeletalMeshLocation002}
   /SetVariables(Particles.SpawnPosition:NiagaraPosition=$Output.SkeletalMeshLocation002.SampledPositionWithoutOffset) {isSetParameters=true}
  ]
  particleUpdateScript [
   &ScaleColor(
    Scale Alpha=&FNTimeOfDayScalar(Night Scale=1, Morning Scale=1, Evening Scale=1)
    Scale RGB=&Multiply_VectorByFloat(Vector=<1 1 1>, Float=&FNTimeOfDayScalar(Night Scale=1.5, Morning Scale=1, Evening Scale=1))
    ScaleA=true
    ScaleRGB=true
   ) {module=ScaleColor001}
   &SkeletalMeshLocation(
    Bone Sampling Mode=::NewEnumerator4
    MaskSpawnGroup=false
    Orientation Sampling=::NewEnumerator1
    Position Sampling=::NewEnumerator2
    Sampled Position Offset=<0 0 -20>
    Sampled Position Offset Coordinate Space:ENiagaraSkelMeshTransforms=::NewEnumerator0
    Skeletal Mesh=DataInterface() {propertyValues {
      DefaultMesh=&SpicyPumpkin_Gadget
      ExcludeBoneName=root
      OptionalComponentClass=null
      PreviewMesh=&SpicyPumpkin_Gadget
      SkinningMode=SkinOnTheFly
      SoftSourceActor=null
      SourceMode=Default
      UvSetIndex=0
      WholeMeshLOD=-1
      bExcludeBone=false
      bReadDeformedGeometry=true
      bRequireCurrentFrameData=true
      ComponentTags []
      FilteredBones [SpicyPumpkin_Pumpkin_1_jnt]
      FilteredSockets []
      MeshUserParameter {parameter {name=None, typeDefHandle {registeredTypeIndex=1}, varData []}}
      SamplingRegions []
     }}
   )
   &SkeletalMeshLocation(
    Bone Sampling Mode=::NewEnumerator4
    MaskSpawnGroup=false
    Orientation Sampling=::NewEnumerator1
    Position Sampling=::NewEnumerator1
    Sampled Position Offset=<0 0 200>
    Sampled Position Offset Coordinate Space:ENiagaraSkelMeshTransforms=::NewEnumerator0
    Skeletal Mesh=DataInterface() {propertyValues {
      DefaultMesh=&SpicyPumpkin_Gadget
      ExcludeBoneName=root
      OptionalComponentClass=null
      PreviewMesh=&SpicyPumpkin_Gadget
      SkinningMode=SkinOnTheFly
      SoftSourceActor=null
      SourceMode=Default
      UvSetIndex=0
      WholeMeshLOD=-1
      bExcludeBone=false
      bReadDeformedGeometry=true
      bRequireCurrentFrameData=true
      ComponentTags []
      FilteredBones [SpicyPumpkin_Pumpkin_1_jnt]
      FilteredSockets []
      MeshUserParameter {parameter {name=None, typeDefHandle {registeredTypeIndex=1}, varData []}}
      SamplingRegions []
     }}
   ) {enabled=false, module=SkeletalMeshLocation001}
   &ParticleState()
   &UpdateMeshOrientation(Rotation Coordinate Space=::NewEnumerator0, Rotation Rate=0.2, Rotation Vector=<0 0 -1>)
   &ScaleMeshSize(Scale Factor=&VectorFromFloat(Value=&FloatFromCurve(FloatCurve=DataInterface() {propertyValues {Curve {keys [<at 0 0.9>,<at 0.45 0.95>,<at 1 1.5>]}}}, Scale Curve=1.25)))
   &DynamicMaterialParameters(
    Index 0 Param 1=&FloatFromCurve(FloatCurve=DataInterface() {propertyValues {Curve {keys [
         {arriveTangent=2.4535, interpMode=RCIM_Cubic, leaveTangent=2.4535, tangentMode=RCTM_User, time=0.017, value=0}
         {interpMode=RCIM_Cubic, tangentMode=RCTM_User, time=0.5, value=0.5}
         {arriveTangent=-2.5865, tangentMode=RCTM_User, time=1, value=0}
        ]}}})
    Module.Index0Param0Write=true
    Module.Index0Param1Write=false
    Module.Index0Param2Write=false
    Module.Index0Param3Write=false
   )
   &InheritSourceMovement(ApplyForce=false, ApplyPosition=true, ApplyPositionFalloff=false) {enabled=false}
   &Drag(
    Drag=&FloatFromCurve(
     FloatCurve=DataInterface() {propertyValues {Curve {defaultValue=1, keys [{tangentMode=RCTM_User, time=0, value=0}, {tangentMode=RCTM_User, time=1, value=1}]}}}
     Scale Curve=2
    )
    Use Linear Drag=true
    Use Rotational Drag=true
   )
   &SolveForcesAndVelocity(Clamp Velocity=false, Limit Acceleration=false)
   &ScaleColor(
    Scale Alpha=&FloatFromCurve(
     FloatCurve=DataInterface() {propertyValues {Curve {keys [
         {arriveTangent=3.7535, interpMode=RCIM_Cubic, leaveTangent=3.7535, tangentMode=RCTM_User, time=0, value=0.4}
         {interpMode=RCIM_Cubic, tangentMode=RCTM_User, time=0.5, value=1}
         {tangentMode=RCTM_User, time=1, value=0}
        ]}}}
     Scale Curve=0.3
    )
    Scale RGB=&MakeVectorFromLinearColorRGB(Color=&FNTimeOfDayColor(
      Night Color=<rgba 0.1253 0.1914 0.5 1>
      Morning Color=<rgba 0.5 0.3383 0.3719 1>
      Day Color=<rgba 0.5 0.5 0.5 1>
      Evening Color=<rgba 0.4474 0.4547 0.4578 1>
     ))
    ScaleA=true
    ScaleRGB=true
   )
   /SetVariables(Particles.Position=&Lerp_Vector(
     A=$Particles.SpawnPosition
     B=&AddVectorToPosition(Position=$Output.SkeletalMeshLocation.SampledPositionWithoutOffset, Vector=<0 0 30>)
     Alpha=$Particles.NormalizedAge
    )) {isSetParameters=true}
  ]
  rendererClasses [&NiagaraMeshRendererProperties]
  renderers [{rendererClass=&NiagaraMeshRendererProperties, rendererIndex=0}]
 }
 Tornado_Wind_Low_Med {
  emitterUpdateScript [
   &EmitterState(Life Cycle Mode=::NewEnumerator1, Loop Duration=10, Scalability Mode=::NewEnumerator1, UseLoopDelay=false)
   &SpawnRate(SpawnRate=10, Use Spawn Probability=false)
  ]
  particleSpawnScript [
   &InitializeParticle(
    Color=&RandomRangeLinearColor(Minimum=<rgba 0.875 0.3047 0.067 0.6>, Maximum=<rgba 0.1198 0.036 0.0082 0.8>)
    Lifetime=1.5
    Mesh Scale=&Multiply_Vector(
     A=&VectorFromFloat(Value=2)
     B=&RandomRangeVector(Minimum=<0.2 0.02 0.15>, Maximum=<0.15 0.02 0.15>)
    )
    Mesh Scale Mode=::NewEnumerator4
    Position=<0 0 -50>
    Position Mode=::NewEnumerator1
    UsePositionOffset=false
    Write Material Random=true
   )
   &CameraOffset(Camera Offset Amount=15)
   &InitialMeshOrientation(
    Mesh Axis To Orient=<0 1 0>
    Mesh Orientation Mode=::NewEnumerator2
    Orientation Coordinate Space=::World
    Orientation Vector=<0 0 1>
    Rotation=&RandomRangeVector(Minimum=<0 0 0>, Maximum=<0 1 0>)
    Use Rotation Vector=true
   )
   &AddVelocity(Velocity=<0 0 107>)
  ]
  particleUpdateScript [
   &ScaleColor(
    Scale Alpha=&FNTimeOfDayScalar(Night Scale=1, Morning Scale=1, Evening Scale=1)
    Scale RGB=&Multiply_VectorByFloat(Vector=<1 1 1>, Float=&FNTimeOfDayScalar(Night Scale=1.5, Morning Scale=1, Evening Scale=1))
    ScaleA=true
    ScaleRGB=true
   ) {module=ScaleColor001}
   &ParticleState()
   &UpdateMeshOrientation(Rotation Coordinate Space=::NewEnumerator0, Rotation Rate=0.2, Rotation Vector=<0 0 -1>)
   &ScaleMeshSize(Scale Factor=&VectorFromFloat(Value=&FloatFromCurve(FloatCurve=DataInterface() {propertyValues {Curve {keys [<at 0 0>,<at 0.45 0.4>,<at 1 1.6>]}}}, Scale Curve=1.25)))
   &DynamicMaterialParameters(
    Index 0 Param 1=&FloatFromCurve(FloatCurve=DataInterface() {propertyValues {Curve {keys [
         {arriveTangent=2.4535, interpMode=RCIM_Cubic, leaveTangent=2.4535, tangentMode=RCTM_User, time=0.017, value=0}
         {interpMode=RCIM_Cubic, tangentMode=RCTM_User, time=0.5, value=0.5}
         {arriveTangent=-2.5865, tangentMode=RCTM_User, time=1, value=0}
        ]}}})
    Module.Index0Param0Write=true
    Module.Index0Param1Write=false
    Module.Index0Param2Write=false
    Module.Index0Param3Write=false
   )
   &InheritSourceMovement(ApplyForce=false, ApplyPosition=true, ApplyPositionFalloff=false) {enabled=false}
   &Drag(
    Drag=&FloatFromCurve(
     FloatCurve=DataInterface() {propertyValues {Curve {defaultValue=1, keys [{tangentMode=RCTM_User, time=0, value=0}, {tangentMode=RCTM_User, time=1, value=1}]}}}
     Scale Curve=2
    )
    Use Linear Drag=true
    Use Rotational Drag=true
   )
   &SolveForcesAndVelocity(Clamp Velocity=false, Limit Acceleration=false)
   &ScaleColor(
    Scale Alpha=&FloatFromCurve(
     FloatCurve=DataInterface() {propertyValues {Curve {keys [
         {arriveTangent=3.7535, interpMode=RCIM_Cubic, leaveTangent=3.7535, tangentMode=RCTM_User, time=0, value=0}
         {interpMode=RCIM_Cubic, tangentMode=RCTM_User, time=0.5, value=1}
         {tangentMode=RCTM_User, time=1, value=0}
        ]}}}
     Scale Curve=0.3
    )
    Scale RGB=&MakeVectorFromLinearColorRGB(Color=&FNTimeOfDayColor(
      Night Color=<rgba 0.1253 0.1914 0.5 1>
      Morning Color=<rgba 0.5 0.3383 0.3719 1>
      Day Color=<rgba 0.5 0.5 0.5 1>
      Evening Color=<rgba 0.4474 0.4547 0.4578 1>
     ))
    ScaleA=true
    ScaleRGB=true
   )
  ]
  rendererClasses [&NiagaraMeshRendererProperties]
  renderers [{rendererClass=&NiagaraMeshRendererProperties, rendererIndex=0}]
 }
 Trailing_Dust {
  emitterUpdateScript [
   &EmitterState(Life Cycle Mode=::NewEnumerator1, Loop Duration=1, UseLoopDelay=true)
   &SpawnRate(SpawnRate=30, Use Spawn Probability=false)
  ]
  particleSpawnScript [
   &InitializeParticle(
    Color=&FNTimeOfDayColor(
     Night Color=<rgba 0.31 0.3405 0.6406 0.3>
     Morning Color=<rgba 0.1458 0.0867 0.0655 0.3>
     Day Color=<rgba 0.1458 0.0867 0.0655 0.3>
     Evening Color=<rgba 0.1458 0.0867 0.0655 0.3>
    )
    Lifetime Max=1
    Lifetime Min=0.5
    Lifetime Mode=::NewEnumerator1
    Position Offset=<0 0 -8>
    Sprite Rotation Angle=&RandomRangeFloat(Minimum=-0.05, Maximum=0.3)
    Sprite Rotation Mode=::NewEnumerator1
    Sprite Size Mode=::NewEnumerator1
    Uniform Sprite Size Max=30
    Uniform Sprite Size Min=40
    UsePositionOffset=true
    Write Material Random=true
   )
   &ShapeLocation(Hemisphere Distribution=<1 0.5>, MaskSpawnGroup=false, Override Seed=false, Sphere Radius=10, UseOwnerScale=true)
   &AddVelocity(Velocity=&RandomRangeVector(Minimum=<-2 -2 6>, Maximum=<2 2 10>, Evaluation Type=::NewEnumerator1)) {enabled=false}
   &InheritVelocity(Module.Limit Speed=true, Velocity Limit=5, Velocity Scale=<0.25 0.25 0.25>) {enabled=false}
   &PointForce(Force Strength=50, Module.Use Falloff=false, Module.UseLinearForce=false)
   &ApplyInitialForces()
  ]
  particleUpdateScript [
   &ParticleState()
   &SubUVAnimation(
    Sub UV Play Rate:NiagaraFloat=&RandomRangeFloat(Minimum=0.03, Maximum=0.05)
    SubUV Animation Mode:ENiagara_SubUVLookupMode=::NewEnumerator3
    SubUV Start Time Offset:NiagaraFloat=&RandomRangeFloat(Minimum=0, Maximum=63)
   )
   &DynamicMaterialParameters(
    Index 0 Param 1=&FloatFromCurve(FloatCurve=DataInterface() {propertyValues {Curve {keys [<at 0 -0.25>,<at 1 -1>]}}})
    Index 0 Param 4=&OneMinusFloat(Float=$Particles.NormalizedAge)
    Module.Param0WriteEnabled=true
    Module.Param1WriteEnabled=false
    Module.Param2WriteEnabled=false
    Module.Param3WriteEnabled=true
   )
   &ScaleSpriteSize(Scale Factor=&Vector2DFromFloat(Value=&FloatFromCurve(
      FloatCurve=DataInterface() {propertyValues {Curve {keys [
          {arriveTangent=1.8751, interpMode=RCIM_Cubic, leaveTangent=1.8751, tangentMode=RCTM_User, time=0, value=0.5}
          {arriveTangent=0.72, interpMode=RCIM_Cubic, leaveTangent=0.72, tangentMode=RCTM_User, time=1, value=1.25}
         ]}}}
      Scale Curve=1.5
     )))
   &Drag(Drag=0.5, Use Linear Drag=true, Use Rotational Drag=false)
   &SolveForcesAndVelocity(Clamp Velocity=false, Limit Acceleration=false)
   &ScaleColor(
    Scale Alpha=&FloatFromCurve(FloatCurve=DataInterface() {propertyValues {Curve {keys [<at 0 1>,<at 0.3 1>,<at 0.85 1>,<at 1 0>]}}})
    ScaleA=true
    ScaleRGB=true
   )
   &ScaleColor(ScaleA=true, ScaleRGB=true) {module=ScaleColor001}
  ]
  rendererClasses [&NiagaraSpriteRendererProperties]
  renderers [{rendererClass=&NiagaraSpriteRendererProperties}]
 }
 Trailing_Leaves {
  simTarget=GPUComputeSim
  emitterUpdateScript [
   &EmitterState(Life Cycle Mode=::NewEnumerator1, UseLoopDelay=false)
   &SpawnPerUnit(Module.Use Max Movement Threshold=true, Module.Use Movement Tolerance=false, Use Spawn Probability=false) {enabled=false}
   &SpawnRate(SpawnRate=20, Use Spawn Probability=false)
  ]
  particleSpawnScript [
   &InitializeParticle(
    Color Maximum=<rgba 0.791 0.3941 0 1>
    Color Minimum=<rgba 0.6573 0.1236 0 1>
    Color Mode=::NewEnumerator2
    Lifetime Min=1.5
    Lifetime Mode=::NewEnumerator1
    Mass Max=1.5
    Mass Mode=::NewEnumerator3
    Position Offset=<0 0 30>
    Sprite Rotation Mode=::NewEnumerator2
    Sprite Size Mode=::NewEnumerator1
    Uniform Sprite Size Max=8
    Uniform Sprite Size Min=4
    UsePositionOffset=true
    Write Material Random=true
   )
   &ShapeLocation(
    Cylinder Height=30
    Cylinder Radius=20
    MaskSpawnGroup=false
    Offset=<0 0 -15>
    Offset Mode=::NewEnumerator0
    Override Seed=false
    Shape Primitive=::NewEnumerator1
    Surface Only=true
    Surface Only Band Thickness=1
    UseOwnerScale=true
   )
   &InitialMeshOrientation(Use Rotation Vector=false)
  ]
  particleUpdateScript [
   &ParticleState()
   &GravityForce()
   &WindForce(
    Friction=0.8
    Offset By Time:Vector3f=<50 10 -125>
    Scale Min / Max:Vector2f=<0.275 1>
    Speed Min / Max:Vector2f=<0 500>
    Turbulence Scale Mode:ENiagaraWindTurbulenceContributionMode=::NewEnumerator1
    Wind Speed=<200 0 0>
    Wind Speed Scale=&RandomRangeFloat(Minimum=0.333, Maximum=1)
   ) {enabled=false}
   &AerodynamicDrag(
    Aerodynamic Drag:NiagaraFloat=&RandomRangeFloat(Minimum=0.9, Maximum=1)
    Lift Contribution=2
    Pivot Offset Maximum:Vector3f=<5 0.5 1>
    Pivot Offset Minimum:Vector3f=<-5 -0.5 -1>
    Rotation Strength:NiagaraFloat=&RandomRangeFloat(Minimum=1.8, Maximum=2)
    Use Linear Drag=true
    Use Rotational Drag=true
   )
   &VortexForce(
    UseOriginOffset=false
    Vortex Axis=<0 0 -1>
    Vortex Force Amount=&MakeCustomFloatFromBool(
     Bool=&SetBoolByFloatComparison(A=&VectorLength(Vector=$User.OwnerLocalVelocity), B=1)
     True Float=&FloatFromCurve(FloatCurve=DataInterface() {propertyValues {Curve {keys [<at 0 1>,<at 0.25 0>]}}}, Scale Curve=0)
     False Float=0
    )
   )
   &Collision(
    Age Colliding Particles=false
    Enable Rest State=false
    Friction=2
    Friction During a Bounce=0.1
    GPU Collision Type=::NewEnumerator2
    KillOnCollision=false
    Module.EnableMaxCollisionCount=false
    Module.Randomize Collision Normal Vector=false
    Particle Radius Scale=0.1
    Restitution=0.05
   )
   &PointForce(
    Force Falloff Distance=75
    Force Strength=350
    Linear Force Vector:Vector3f=&TransformVector(Vector=<0 3 3>, Destination Space=::Simulation)
    Module.Use Falloff=true
    Module.UseLinearForce=true
   ) {enabled=false}
   &InheritVelocity(
    Inherited Velocity Amount Scale=&VectorFromCurve(VectorCurve=DataInterface() {propertyValues {XCurve {keys [<at 0 1>,<at 0.6 0>]}, YCurve {keys [<at 0 1>,<at 0.6 0>]}, ZCurve {keys [<at 0 1>,<at 0.6 0>]}}})
    Speed Limit=true
    UseSpeedThreshold=false
   )
   &SolveForcesAndVelocity(Clamp Velocity=false, Limit Acceleration=false)
   &ScaleSpriteSize(Uniform Curve Sprite Scale=DataInterface() {propertyValues {Curve {keys [
        {arriveTangent=14.5835, interpMode=RCIM_Cubic, leaveTangent=14.5835, tangentMode=RCTM_User, time=0, value=0}
        {arriveTangent=1.5113, interpMode=RCIM_Cubic, leaveTangent=1.5113, time=0.1, value=1}
        {arriveTangent=-1.1111, interpMode=RCIM_Cubic, leaveTangent=-1.1111, time=0.7087, value=1.071}
        {arriveTangent=-8.7113, interpMode=RCIM_Cubic, leaveTangent=-8.7113, tangentMode=RCTM_User, time=1, value=0}
       ]}}})
   &AlignSpriteToMeshOrientation(SetAlignment=true, SetFacing=true)
   &AlignParticlesWithCollisionPlane(Allow Alignment on Y=false, Allow Alignment on Z=false, Rotational Dampening Power=1)
   &ScaleColor(
    Scale Alpha=&MakeCustomFloatFromBool(
     Bool=&SetBoolByFloatComparison(A=&VectorLength(Vector=$User.OwnerLocalVelocity), B=1)
     True Float=&FloatFromCurve(FloatCurve=DataInterface() {propertyValues {Curve {keys [<at 0 1>,<at 0.3 1>,<at 0.8 1>,<at 0.9 0>]}}})
     False Float=&FloatFromCurve(FloatCurve=DataInterface() {propertyValues {Curve {keys [<at 0 0>,<at 0.4117 0.1692>,<at 0.5754 0.9431>,<at 0.882 1>,<at 1 0>]}}})
    )
    Scale RGB=&SelectVectorFromArray(
     Array Sampling Mode=::NewEnumerator0
     Direct Array Index=?
     Vector Selection Array=DataInterface() {propertyValues {FloatData [<1 1 1>, <2 2 2>]}}
     Random Seed=0
    )
    ScaleA=true
    ScaleRGB=true
   )
  ]
  rendererClasses [&NiagaraSpriteRendererProperties, &NiagaraMeshRendererProperties]
  renderers [
   {rendererClass=&NiagaraSpriteRendererProperties}
   {rendererClass=&NiagaraMeshRendererProperties, rendererIndex=1}
  ]
 }
 Trailing_Wind {
  emitterUpdateScript [
   &EmitterState(Life Cycle Mode=::NewEnumerator1, Loop Duration=10, Scalability Mode=::NewEnumerator1, UseLoopDelay=false)
   &SpawnRate(SpawnRate=10, Use Spawn Probability=false)
  ]
  particleSpawnScript [
   &InitializeParticle(
    Color=&RandomRangeLinearColor(Minimum=<rgba 0.599 0.3464 0.2947 0.5>, Maximum=<rgba 0.5052 0.2122 0.2122 0.8>)
    Lifetime=0.5
    Mesh Scale Max=<0.2 0.1 0.2>
    Mesh Scale Min=<0.25 0.1 0.2>
    Mesh Scale Mode=::NewEnumerator2
    Position Mode=::NewEnumerator1
    UsePositionOffset=false
    Write Material Random=true
   )
   &CameraOffset(Camera Offset Amount=15)
   &InitialMeshOrientation(
    Mesh Axis To Orient=<0 1 0>
    Mesh Orientation Mode=::NewEnumerator2
    Orientation Coordinate Space=::World
    Orientation Vector=<0 0 1>
    Rotation=&RandomRangeVector(Minimum=<0 0 0>, Maximum=<0 1 0>)
    Use Rotation Vector=true
   )
   &AddVelocity(Velocity=<0 0 80>)
  ]
  particleUpdateScript [
   &ScaleColor(
    Scale Alpha=&FN_Select_Float_from_Time_Of_Day(Morning=1, Day=1, Evening=1, Night=2)
    Scale RGB=&FN_Select_Vector_from_Time_Of_Day(Morning=<1.5 0.5 0.5>, Day=<1 1 1>, Evening=<1.5 0.5 0.5>, Night=<0.7 0.7 2>)
    ScaleA=true
    ScaleRGB=true
   ) {module=ScaleColor001}
   &ParticleState()
   &ScaleMeshSize(Scale Factor=&VectorFromFloat(Value=&FloatFromCurve(FloatCurve=DataInterface() {propertyValues {Curve {keys [<at 0 0.7>,<at 0.45 0.875>,<at 1 1.5>]}}}, Scale Curve=1.25)))
   &DynamicMaterialParameters(
    Index 0 Param 1=&FloatFromCurve(FloatCurve=DataInterface() {propertyValues {Curve {keys [
         {arriveTangent=2.4535, interpMode=RCIM_Cubic, leaveTangent=2.4535, tangentMode=RCTM_User, time=0, value=0}
         {interpMode=RCIM_Cubic, tangentMode=RCTM_User, time=0.5, value=0.5}
         {arriveTangent=-2.5865, tangentMode=RCTM_User, time=1, value=0}
        ]}}})
    Module.Index0Param0Write=true
    Module.Index0Param1Write=false
    Module.Index0Param2Write=false
    Module.Index0Param3Write=false
   )
   &InheritSourceMovement(ApplyForce=false, ApplyPosition=true, ApplyPositionFalloff=false) {enabled=false}
   &SolveForcesAndVelocity(Clamp Velocity=false, Limit Acceleration=false)
   &ScaleColor(
    Scale Alpha=&FloatFromCurve(
     FloatCurve=DataInterface() {propertyValues {Curve {keys [
         {arriveTangent=6.3511, interpMode=RCIM_Cubic, leaveTangent=6.3511, tangentMode=RCTM_User, time=0, value=0}
         {interpMode=RCIM_Cubic, tangentMode=RCTM_User, time=0.5, value=1}
         {tangentMode=RCTM_User, time=1, value=0}
        ]}}}
     Scale Curve=0.3
    )
    Scale RGB=&MakeVectorFromLinearColorRGB(Color=&FNTimeOfDayColor(
      Night Color=<rgba 0.1253 0.1914 0.5 1>
      Morning Color=<rgba 0.5 0.3383 0.3719 1>
      Day Color=<rgba 0.5 0.5 0.5 1>
      Evening Color=<rgba 0.4474 0.4547 0.4578 1>
     ))
    ScaleA=true
    ScaleRGB=true
   )
  ]
  rendererClasses [&NiagaraMeshRendererProperties]
  renderers [{rendererClass=&NiagaraMeshRendererProperties, rendererIndex=0}]
 }
}
summary {
 systemName=NS_Emote_SpicyPumpkin_Leaves
 emitters [
  {emitterName=Trailing_Dust, rendererClasses [&NiagaraSpriteRendererProperties]}
  {
   emitterName=Trailing_Leaves
   simTarget=GPUComputeSim
   rendererClasses [&NiagaraSpriteRendererProperties, &NiagaraMeshRendererProperties]
  }
  {emitterName=Light_Wind, rendererClasses [&NiagaraMeshRendererProperties]}
  {emitterName=Trailing_Wind, rendererClasses [&NiagaraMeshRendererProperties]}
  {emitterName=Local_Dust, rendererClasses [&NiagaraSpriteRendererProperties]}
  {emitterName=Skeletal_Wind, rendererClasses [&NiagaraMeshRendererProperties]}
  {
   emitterName=Leaf_Tornado_Bottom
   simTarget=GPUComputeSim
   rendererClasses [&NiagaraSpriteRendererProperties, &NiagaraMeshRendererProperties]
  }
  {emitterName=Dark_Wind, rendererClasses [&NiagaraMeshRendererProperties]}
  {
   emitterName=Leaf_Tornado_Top
   simTarget=GPUComputeSim
   rendererClasses [&NiagaraSpriteRendererProperties, &NiagaraMeshRendererProperties]
  }
  {emitterName=Tornado_Wind_Low_Med, rendererClasses [&NiagaraMeshRendererProperties]}
 ]
 userVariables [
  User.BaseColorAlpha {type=LinearColor, description="", defaultValue {struct=LinearColor, value=<rgba 1 1 1 0.1>}}
  User.DeltaTimeMax {type=NiagaraFloat, defaultValue=0.0167, description=""}
  User.EmissiveColor {type=LinearColor, description="", defaultValue {struct=LinearColor, value=<rgba 1 1 1 0.761>}}
  User.FixedDeltaTime {type=NiagaraFloat, defaultValue=0, description=""}
  User.Global_Smoke_Opacity_Multiplier {type=NiagaraFloat, defaultValue=0, description=""}
  User.NormalizedNotifyProgress {type=NiagaraFloat, defaultValue=0, description=""}
  User.OwnerLocalVelocity {type=Vector3f, description="", defaultValue {struct=Vector3f, value=<0 0 0>}}
  User.RibbonScale {type=NiagaraFloat, defaultValue=1.5, description=""}
  User.Smoke_Alpha {type=NiagaraFloat, defaultValue=2, description=""}
  User.Smoke_Position_Offset {type=NiagaraFloat, defaultValue=0, description=""}
  User.UseEmissiveSmoke {type=NiagaraBool}
 ]
}
