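# Niagara system for the Spicy Pumpkin emote, adding steam and sparkles to the coffee mug gadget.
# Flipbook steam sprites and two warm sparkle emitters spawn around the mug's steam socket, sampled from the attach-parent skeletal mesh.
# A rising ribbon trail adds time-of-day tint and optional emissive color.
# Owner-speed thresholds switch ribbon lifetime, trailing-sparkle spawn rate and both trailing emitters' world/local BlendSpace, which an animation inheritance parameter also drives.
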NiagaraSystem /BRCosmetics/Animation/Game/MainPlayer/Emotes/SpicyPumpkin/FX/NS_Emote_SpicyPumpkin_Steam
dependencies {
 usedDataInterfaces [&NiagaraDataInterfaceSkeletalMesh, &NiagaraDataInterfaceCurve]
 usedDynamicInputs []
 usedModules [&SystemState, &BlendSpaceHelper]
}
emitters {
 Ambient_Steam {
  emitterUpdateScript [
   &EmitterState(Life Cycle Mode=::NewEnumerator1, Loop Duration=1, UseLoopDelay=true)
   &SpawnRate(SpawnRate=3, Use Spawn Probability=false)
  ]
  particleSpawnScript [
   &InitializeParticle(
    Color=<rgba 0.6875 0.6875 0.6875 0.05>
    Lifetime Max=3
    Lifetime Min=2.3
    Lifetime Mode=::NewEnumerator1
    Position Offset=<0 0 5>
    Sprite Rotation Angle=&RandomRangeFloat(Minimum=-0.05, Maximum=0.3)
    Sprite Rotation Mode=::NewEnumerator1
    Sprite Size Mode=::NewEnumerator1
    Uniform Sprite Size Max=15
    Uniform Sprite Size Min=20
    UsePositionOffset=true
    Write Material Random=true
   )
   &SkeletalMeshLocation(
    MaskSpawnGroup=false
    Mesh Sampling Type=::NewEnumerator5
    Position Sampling=::NewEnumerator2
    Sampled Position Offset Coordinate Space:ENiagaraSkelMeshTransforms=::NewEnumerator0
    Skeletal Mesh=DataInterface() {propertyValues {
      DefaultMesh=/BRCosmetics/Animation/Props/Gadgets_SpicyPumpkin/Meshes/SpicyPumpkin_Gadget
      ExcludeBoneName=root
      OptionalComponentClass=null
      PreviewMesh=&SpicyPumpkin_Gadget
      SkinningMode=SkinOnTheFly
      SoftSourceActor=null
      SourceMode=AttachParent
      UvSetIndex=0
      WholeMeshLOD=-1
      bExcludeBone=false
      bReadDeformedGeometry=true
      bRequireCurrentFrameData=true
      ComponentTags []
      FilteredBones []
      FilteredSockets [SpicyPumpkin_Coffee_mug_1__mug_steam]
      MeshUserParameter {parameter {name=None, typeDefHandle {registeredTypeIndex=1}, varData []}}
      SamplingRegions []
     }}
    Socket Sampling Mode=::NewEnumerator5
   )
   &SystemLocation(
    MaskSpawnGroup=false
    Offset=&Lerp_Vector(
     A=<0 0 0>
     B=&Subtract_Vector(A=<0.0003 1.9429 -9.3009>, B=<0.0003 -6.7513 -9.5841>)
     Alpha=$User.Smoke_Position_Offset
    )
   ) {enabled=false}
   &ShapeLocation(Apply To Particle Position=::NewEnumerator1, MaskSpawnGroup=false, Override Seed=false, Sphere Radius=10, UseOwnerScale=true)
   /SetVariables(Particles.Position=&AddVectorToPosition(
     Position=&AddVectorToPosition(
      Position=$Particles.SkeletalMeshLocation.SampledPosition
      Vector=Linked() {linkedVariable {name=Particles.ShapeLocation.ShapePosition, type=NiagaraPosition}}
     )
     Vector=<0 0 -8>
    )) {isSetParameters=true}
   &AddVelocity(Velocity=&RandomRangeVector(Minimum=<-2 -2 6>, Maximum=<2 2 10>, Evaluation Type=::NewEnumerator1))
   &InheritVelocity(Module.Limit Speed=true, Velocity Limit=5, Velocity Scale=<0.25 0.25 0.25>)
  ]
  particleUpdateScript [
   &ParticleState()
   &SubUVAnimation(
    Sub UV Play Rate:NiagaraFloat=&RandomRangeFloat(Minimum=0.03, Maximum=0.05)
    SubUV Animation Mode:ENiagara_SubUVLookupMode=::NewEnumerator3
    SubUV Start Time Offset:NiagaraFloat=&RandomRangeFloat(Minimum=0, Maximum=63)
   )
   &DynamicMaterialParameters(
    Index 0 Param 1=&FloatFromCurve(FloatCurve=DataInterface() {propertyValues {Curve {keys [<at 0 -0.25>,<at 1 -1>]}}})
    Index 0 Param 4=&OneMinusFloat(Float=$Particles.NormalizedAge)
    Module.Param0WriteEnabled=true
    Module.Param1WriteEnabled=false
    Module.Param2WriteEnabled=false
    Module.Param3WriteEnabled=true
   )
   &ScaleSpriteSize(Scale Factor=&Vector2DFromFloat(Value=&FloatFromCurve(
      FloatCurve=DataInterface() {propertyValues {Curve {keys [
          {arriveTangent=1.8751, interpMode=RCIM_Cubic, leaveTangent=1.8751, tangentMode=RCTM_User, time=0, value=0.25}
          {arriveTangent=0.72, interpMode=RCIM_Cubic, leaveTangent=0.72, tangentMode=RCTM_User, time=1, value=1.25}
         ]}}}
      Scale Curve=1.5
     )))
   &Drag(Drag=0.5, Use Linear Drag=true, Use Rotational Drag=false)
   &InheritVelocity(Module.Limit Speed=true, Velocity Limit=0.3, Velocity Scale=&VectorFromFloat(Value=0.01)) {module=InheritVelocity001}
   &SolveForcesAndVelocity(Clamp Velocity=false, Limit Acceleration=false)
   &ScaleColor(
    Scale Alpha=&FloatFromCurve(FloatCurve=DataInterface() {propertyValues {Curve {keys [<at 0 1>,<at 0.3 1>,<at 0.85 1>,<at 1 0>]}}})
    ScaleA=true
    ScaleRGB=true
   )
   &ScaleColor(ScaleA=true, ScaleRGB=true) {module=ScaleColor001}
  ]
  rendererClasses [&NiagaraSpriteRendererProperties]
  renderers [{rendererClass=&NiagaraSpriteRendererProperties}]
 }
 Local_Sparkles {
  emitterUpdateScript [
   &EmitterState(Life Cycle Mode=::NewEnumerator1, Loop Duration=2, UseLoopDelay=false)
   &SpawnRate(SpawnRate=5, Use Spawn Probability=false)
  ]
  particleSpawnScript [
   &SkeletalMeshLocation(
    MaskSpawnGroup=false
    Mesh Sampling Type=::NewEnumerator5
    Position Sampling=::NewEnumerator2
    Sampled Position Offset Coordinate Space:ENiagaraSkelMeshTransforms=::NewEnumerator1
    Skeletal Mesh=DataInterface() {propertyValues {
      DefaultMesh=&SpicyPumpkin_Gadget
      ExcludeBoneName=root
      OptionalComponentClass=null
      PreviewMesh=&SpicyPumpkin_Gadget
      SkinningMode=SkinOnTheFly
      SoftSourceActor=null
      SourceMode=AttachParent
      UvSetIndex=0
      WholeMeshLOD=-1
      bExcludeBone=false
      bReadDeformedGeometry=true
      bRequireCurrentFrameData=true
      ComponentTags []
      FilteredBones []
      FilteredSockets [SpicyPumpkin_Coffee_mug_1__mug_steam]
      MeshUserParameter {parameter {name=None, typeDefHandle {registeredTypeIndex=1}, varData []}}
      SamplingRegions []
     }}
    Socket Sampling Mode=::NewEnumerator5
   )
   &InitializeParticle(
    Color=<rgba 1 0.7027 0.3594 1>
    Lifetime Max=1
    Lifetime Min=0.8
    Lifetime Mode=::NewEnumerator1
    Position Offset=<0 0 10>
    Sprite Size Mode=::NewEnumerator1
    Uniform Sprite Size Max=30
    Uniform Sprite Size Min=20
    UsePositionOffset=true
    Write Material Random=true
   )
   &ShapeLocation(Apply To Particle Position=::NewEnumerator1, MaskSpawnGroup=false, Override Seed=false, Sphere Radius=10, UseOwnerScale=true)
   /SetVariables(Particles.Position=&AddVectorToPosition(
     Position=&AddVectorToPosition(
      Position=$Particles.SkeletalMeshLocation.SampledPosition
      Vector=Linked() {linkedVariable {name=Particles.ShapeLocation.ShapePosition, type=NiagaraPosition}}
     )
     Vector=<0 0 -8>
    )) {isSetParameters=true, module=SetVariables_#1}
   &CameraOffset(Camera Offset Amount=-10)
  ]
  particleUpdateScript [
   &ParticleState()
   &WindForce(Frequency=6, Offset By Time:Vector3f=<0 0 -50>, Scale=4, Wind Speed=<1 0 0>)
   &AerodynamicDrag(
    Aerodynamic Drag:NiagaraFloat=&RandomRangeFloat(Minimum=0.6, Maximum=1)
    Aerodynamic Rotational Drag=0.2
    Drag Shape Curve=DataInterface() {propertyValues {Curve {keys [
        {interpMode=RCIM_Cubic, tangentMode=RCTM_User, time=0, value=0.1}
        {arriveTangent=2.0799, interpMode=RCIM_Cubic, leaveTangent=2.0799, tangentMode=RCTM_User, time=1, value=1}
       ]}}}
    Lift Shape Curve=DataInterface() {propertyValues {Curve {keys [
        {interpMode=RCIM_Cubic, tangentMode=RCTM_User, time=0, value=0}
        {interpMode=RCIM_Cubic, tangentMode=RCTM_User, time=0.5, value=1}
        {tangentMode=RCTM_User, time=1, value=0}
       ]}}}
    Use Linear Drag=true
    Use Rotational Drag=true
   )
   &ScaleSpriteSize(Scale Factor=&Vector2DFromFloat(Value=&FloatFromCurve(FloatCurve=DataInterface() {propertyValues {Curve {
         defaultValue=1
         keys [{interpMode=RCIM_Cubic, tangentMode=RCTM_User, time=0, value=1}, {interpMode=RCIM_Cubic, tangentMode=RCTM_User, time=1, value=0}]
        }}})))
   &ScaleColor(
    Scale Alpha=&FloatFromCurve(FloatCurve=DataInterface() {propertyValues {Curve {keys [<at 0 0 RCIM_Cubic>,<at 0.5 1 RCIM_Cubic>,<at 1 0 RCIM_Cubic>]}}})
    ScaleA=true
    ScaleRGB=true
   )
   &SolveForcesAndVelocity(Clamp Velocity=false, Limit Acceleration=false)
  ]
  rendererClasses [&NiagaraSpriteRendererProperties]
  renderers [{rendererClass=&NiagaraSpriteRendererProperties}]
 }
 Steam_Trail {
  emitterUpdateScript [
   &EmitterState(Life Cycle Mode=::NewEnumerator1, Loop Duration=1, UseLoopDelay=false)
   /SetVariables(Emitter.NewSpawnRate:NiagaraFloat=&Lerp_Float(A=20, B=20, Alpha=$User.Anim_LocalSpaceInheritance)) {isSetParameters=true}
   &SpawnPerUnit(Module.Use Max Movement Threshold=true, Module.Use Movement Tolerance=false, Spawn Spacing=10, Use Spawn Probability=false) {enabled=false}
   &SpawnRate(SpawnRate=$Emitter.NewSpawnRate, Use Spawn Probability=false)
  ]
  particleSpawnScript [
   &SkeletalMeshLocation(
    MaskSpawnGroup=false
    Mesh Sampling Type=::NewEnumerator5
    Orientation Sampling=::NewEnumerator1
    Position Sampling=::NewEnumerator2
    Sampled Position Offset Coordinate Space:ENiagaraSkelMeshTransforms=::NewEnumerator0
    Skeletal Mesh=DataInterface() {propertyValues {
      DefaultMesh=&SpicyPumpkin_Gadget
      ExcludeBoneName=root
      OptionalComponentClass=null
      PreviewMesh=&SpicyPumpkin_Gadget
      SkinningMode=SkinOnTheFly
      SoftSourceActor=null
      SourceMode=AttachParent
      UvSetIndex=0
      WholeMeshLOD=-1
      bExcludeBone=false
      bReadDeformedGeometry=true
      bRequireCurrentFrameData=true
      ComponentTags []
      FilteredBones [SpicyPumpkin_Coffee_mug_2_jnt]
      FilteredSockets [SpicyPumpkin_Coffee_mug_1__mug_steam]
      MeshUserParameter {parameter {name=None, typeDefHandle {registeredTypeIndex=1}, varData []}}
      SamplingRegions []
     }}
    Socket Sampling Mode=::NewEnumerator5
   ) {enabled=false}
   &SkeletalMeshLocation(
    Bone Sampling Mode=::NewEnumerator4
    MaskSpawnGroup=false
    Orientation Sampling=::NewEnumerator1
    Position Sampling=::NewEnumerator2
    Sampled Position Offset=<0 0 10>
    Sampled Position Offset Coordinate Space:ENiagaraSkelMeshTransforms=::NewEnumerator0
    Skeletal Mesh=DataInterface() {propertyValues {
      DefaultMesh=null
      ExcludeBoneName=root
      OptionalComponentClass=null
      PreviewMesh=&SpicyPumpkin_Gadget
      SkinningMode=SkinOnTheFly
      SoftSourceActor=null
      SourceMode=AttachParent
      UvSetIndex=0
      WholeMeshLOD=-1
      bExcludeBone=false
      bReadDeformedGeometry=true
      bRequireCurrentFrameData=true
      ComponentTags []
      FilteredBones [root]
      FilteredSockets [SpicyPumpkin_Coffee_mug_1__mug_steam]
      MeshUserParameter {parameter {name=None, typeDefHandle {registeredTypeIndex=1}, varData []}}
      SamplingRegions []
     }}
   ) {module=SkeletalMeshLocation001}
   &InitializeParticle(
    Color=&FNTimeOfDayColor(
     Night Color=<rgba 0.1929 0.2808 1 0.1>
     Morning Color=<rgba 1 0.3787 0.4169 0.1>
     Day Color=<rgba 1 1 1 0.1>
     Evening Color=<rgba 1 0.4427 0.529 0.1>
    )
    Lifetime=&MakeCustomFloatFromBool(
     Bool=&SetBoolByFloatComparison(A=&VectorLength(Vector=$User.OwnerLocalVelocity), B=300)
     True Float=&Lerp_Float(A=0.1, B=0.2, Alpha=$User.Anim_LocalSpaceInheritance)
     False Float=&Lerp_Float(A=0.5, B=0.22, Alpha=$User.Anim_LocalSpaceInheritance)
    )
    Ribbon Facing Mode=::NewEnumerator1
    Ribbon Facing Vector:Vector3f=<1 0 0>
    Ribbon Width=&Multiply_Float(A=4, B=$User.RibbonScale)
    Ribbon Width Mode=::NewEnumerator1
    UsePositionOffset=true
    Write Material Random=true
   )
   /SetVariables(Particles.Position=&AddVectorToPosition(Position=$Particles.SkeletalMeshLocation.SampledPosition, Vector=<0 0 0>)) {enabled=false, isSetParameters=true}
   &AddVelocity(
    Coordinate Space=::World
    Velocity=&Multiply_VectorByFloat(Vector=<0 0 60>, Float=&Lerp_Float(A=0.7, B=1, Alpha=$User.Anim_LocalSpaceInheritance))
   )
  ]
  particleUpdateScript [
   &ScaleColor(Scale Alpha=2, ScaleA=true, ScaleRGB=true) {module=ScaleColor003}
   &ParticleState()
   &self:Leading_Edge_Fix_01(Position:NiagaraPosition=$Particles.SkeletalMeshLocation.SampledPosition) {module=Leading_Edge_Fix_01}
   &ScaleColor(
    Scale Alpha=&Multiply_Float(
     A=&FloatFromCurve(
      FloatCurve=DataInterface() {propertyValues {Curve {keys [<at 0 1>, {arriveTangent=-1.25, interpMode=RCIM_Cubic, leaveTangent=-1.25, time=0.4152, value=1}, <at 0.8 0>]}}}
      Scale Curve=&FNTimeOfDayScalar(Night Scale=1, Morning Scale=0.3, Evening Scale=0.5)
     )
     B=&FloatFromCurve(FloatCurve=DataInterface() {propertyValues {Curve {keys [<at 0 1>,<at 0.3 1>,<at 0.85 1>,<at 1 0>]}}})
    )
    Scale RGB=&MakeVectorFromLinearColorRGB(Color=&MakeCustomLinearColorFromBool(
      Bool=$User.UseEmissiveSmoke
      True Color=&LerpLinearColors(
       StartColor=$User.EmissiveColor
       EndColor=&MultiplyLinearColor(Linear Color=$User.EmissiveColor, Scale Factor=&MakeVector4_FromFloat(Float=0.25))
       LerpFactor=&FloatFromCurve(FloatCurve=DataInterface() {propertyValues {Curve {defaultValue=0, keys [{tangentMode=RCTM_User, time=0, value=1}, {tangentMode=RCTM_User, time=1, value=1}]}}})
      )
      False Color=<rgba 1 1 1 1>
     ))
    ScaleA=true
    ScaleRGB=true
   )
   &DynamicMaterialParameters(
    Index 0 Param 1=&FloatFromCurve(FloatCurve=DataInterface() {propertyValues {Curve {keys [
         <at 0 1 RCIM_Cubic>
         {arriveTangent=-1.7387, arriveTangentWeight=1.9926, interpMode=RCIM_Cubic, leaveTangent=-1.7387, tangentMode=RCTM_User, time=1, value=0.125}
        ]}}})
    Index 0 Param 3=-1
    Index 0 Param 4=&MakeCustomFloatFromBool(
     Bool=$User.UseEmissiveSmoke
     True Float=&FloatFromCurve(FloatCurve=DataInterface() {propertyValues {Curve {keys [{arriveTangent=-2.94, interpMode=RCIM_Cubic, time=0, value=1}, <at 0.75 0 RCIM_Cubic>]}}})
     False Float=0
    )
    Index 1 Param 1=&FloatFromCurve(FloatCurve=DataInterface() {propertyValues {Curve {keys [<at 0 1 RCIM_Cubic>]}}})
    Index 1 Param 2=&FloatFromCurve(FloatCurve=DataInterface() {propertyValues {Curve {keys [{arriveTangent=-2.2405, interpMode=RCIM_Cubic, leaveTangent=-2.2405, tangentMode=RCTM_User, time=0, value=1}, <at 1 0 RCIM_Cubic>]}}})
    Module.Index0Float0Write=true
    Module.Index0Float1Write=true
    Module.Index0Float2Write=true
    Module.Index0Float3Write=true
    Module.Index1Float0Write=true
    Module.Index1Float1Write=true
    Module.Index1Float2Write=true
    Module.Index1Float3Write=true
    Write Parameter Index 1=::NewEnumerator1
   ) {enabled=false}
   &ScaleRibbonWidth(Ribbon Width Scale=&FloatFromCurve(
     FloatCurve=DataInterface() {propertyValues {Curve {keys [{tangentMode=RCTM_User, time=0, value=1}, {tangentMode=RCTM_User, time=1, value=0}]}}}
     Scale Curve=3
    ))
   &CurlNoiseForce(Bias Noise=false, Noise Strength=50, Pan Noise=false) {enabled=false}
   &SolveForcesAndVelocity(Clamp Velocity=false, Limit Acceleration=false)
   &ScaleColor(ScaleA=true, ScaleRGB=true) {module=ScaleColor001}
   &BlendSpace(
    Mesh Orientation Mode=1
    "Position Blend (World <-> Local)":NiagaraFloat=&MakeCustomFloatFromBool(
     Bool=&SetBoolByFloatComparison(A=&VectorLength(Vector=$User.OwnerLocalVelocity), B=300)
     True Float=0
     False Float=&Lerp_Float(
      A=0.9
      B=&FloatFromCurve(FloatCurve=DataInterface() {propertyValues {Curve {defaultValue=1, keys [{tangentMode=RCTM_User, time=0, value=0}, <at 0.7376 0.2618>, {tangentMode=RCTM_User, time=1, value=1}]}}})
      Alpha=$User.Anim_LocalSpaceInheritance
     )
     Translation Contribution:NiagaraFloat=1
     Rotation Contribution:NiagaraFloat=1
    )
    Position Coordinate Space=2
    Velocity Mode=1
   )
   &ScaleColor(Scale Alpha=$User.Global_Smoke_Opacity_Multiplier, ScaleA=true, ScaleRGB=true) {enabled=false, module=ScaleColor002}
  ]
  rendererClasses [&NiagaraRibbonRendererProperties]
  renderers [{rendererClass=&NiagaraRibbonRendererProperties}]
 }
 Trailing_Sparkles {
  emitterUpdateScript [
   &EmitterState(Life Cycle Mode=::NewEnumerator1, Loop Duration=2, UseLoopDelay=false)
   &SpawnRate(
    SpawnRate=&MakeCustomFloatFromBool(
     Bool=&SetBoolByFloatComparison(A=&VectorLength(Vector=$User.OwnerLocalVelocity), B=200)
     True Float=0
     False Float=10
    )
    Use Spawn Probability=false
   )
  ]
  particleSpawnScript [
   &SkeletalMeshLocation(
    MaskSpawnGroup=false
    Mesh Sampling Type=::NewEnumerator5
    Position Sampling=::NewEnumerator2
    Sampled Position Offset Coordinate Space:ENiagaraSkelMeshTransforms=::NewEnumerator1
    Skeletal Mesh=DataInterface() {propertyValues {
      DefaultMesh=&SpicyPumpkin_Gadget
      ExcludeBoneName=root
      OptionalComponentClass=null
      PreviewMesh=&SpicyPumpkin_Gadget
      SkinningMode=SkinOnTheFly
      SoftSourceActor=null
      SourceMode=AttachParent
      UvSetIndex=0
      WholeMeshLOD=-1
      bExcludeBone=false
      bReadDeformedGeometry=true
      bRequireCurrentFrameData=true
      ComponentTags []
      FilteredBones []
      FilteredSockets [SpicyPumpkin_Coffee_mug_1__mug_steam]
      MeshUserParameter {parameter {name=None, typeDefHandle {registeredTypeIndex=1}, varData []}}
      SamplingRegions []
     }}
    Socket Sampling Mode=::NewEnumerator5
   )
   &InitializeParticle(
    Color=<rgba 1 0.7027 0.3594 1>
    Lifetime Max=0.4
    Lifetime Min=0.3
    Lifetime Mode=::NewEnumerator1
    Position Offset=<0 0 10>
    Sprite Size Mode=::NewEnumerator1
    Uniform Sprite Size Max=30
    Uniform Sprite Size Min=20
    UsePositionOffset=true
    Write Material Random=true
   )
   &ShapeLocation(Apply To Particle Position=::NewEnumerator1, MaskSpawnGroup=false, Override Seed=false, Sphere Radius=10, UseOwnerScale=true)
   /SetVariables(Particles.Position=&AddVectorToPosition(
     Position=&AddVectorToPosition(
      Position=$Particles.SkeletalMeshLocation.SampledPosition
      Vector=Linked() {linkedVariable {name=Particles.ShapeLocation.ShapePosition, type=NiagaraPosition}}
     )
     Vector=<0 0 -8>
    )) {isSetParameters=true, module=SetVariables_#1}
   &CameraOffset(Camera Offset Amount=-50)
   &AddVelocity(Velocity=<0 0 30>) {enabled=false}
  ]
  particleUpdateScript [
   &ParticleState()
   &WindForce(Frequency=6, Offset By Time:Vector3f=<0 0 -50>, Scale=4, Wind Speed=<1 0 0>)
   &AerodynamicDrag(
    Aerodynamic Drag:NiagaraFloat=&RandomRangeFloat(Minimum=0.6, Maximum=1)
    Aerodynamic Rotational Drag=0.2
    Drag Shape Curve=DataInterface() {propertyValues {Curve {keys [
        {interpMode=RCIM_Cubic, tangentMode=RCTM_User, time=0, value=0.1}
        {arriveTangent=2.0799, interpMode=RCIM_Cubic, leaveTangent=2.0799, tangentMode=RCTM_User, time=1, value=1}
       ]}}}
    Lift Shape Curve=DataInterface() {propertyValues {Curve {keys [
        {interpMode=RCIM_Cubic, tangentMode=RCTM_User, time=0, value=0}
        {interpMode=RCIM_Cubic, tangentMode=RCTM_User, time=0.5, value=1}
        {tangentMode=RCTM_User, time=1, value=0}
       ]}}}
    Use Linear Drag=true
    Use Rotational Drag=true
   )
   &ScaleSpriteSize(Scale Factor=&Vector2DFromFloat(Value=&FloatFromCurve(FloatCurve=DataInterface() {propertyValues {Curve {
         defaultValue=1
         keys [{interpMode=RCIM_Cubic, tangentMode=RCTM_User, time=0, value=1}, {interpMode=RCIM_Cubic, tangentMode=RCTM_User, time=1, value=0}]
        }}})))
   &ScaleColor(
    Scale Alpha=&FloatFromCurve(FloatCurve=DataInterface() {propertyValues {Curve {keys [<at 0 0 RCIM_Cubic>,<at 0.5 1 RCIM_Cubic>,<at 1 0 RCIM_Cubic>]}}})
    ScaleA=true
    ScaleRGB=true
   )
   &SolveForcesAndVelocity(Clamp Velocity=false, Limit Acceleration=false)
   &BlendSpace(
    Mesh Orientation Mode=1
    "Position Blend (World <-> Local)":NiagaraFloat=&MakeCustomFloatFromBool(
     Bool=&SetBoolByFloatComparison(A=&VectorLength(Vector=$User.OwnerLocalVelocity), B=300)
     True Float=1
     False Float=&Lerp_Float(
      A=0.9
      B=&FloatFromCurve(FloatCurve=DataInterface() {propertyValues {Curve {defaultValue=1, keys [{tangentMode=RCTM_User, time=0, value=0}, <at 0.5151 0.6305>, <at 1 1>]}}})
      Alpha=$User.Anim_LocalSpaceInheritance
     )
     Translation Contribution:NiagaraFloat=1
     Rotation Contribution:NiagaraFloat=1
    )
    Position Coordinate Space=2
    Velocity Mode=1
   )
  ]
  rendererClasses [&NiagaraSpriteRendererProperties]
  renderers [{rendererClass=&NiagaraSpriteRendererProperties}]
 }
}
summary {
 systemName=NS_Emote_SpicyPumpkin_Steam
 emitters [
  {emitterName=Ambient_Steam, rendererClasses [&NiagaraSpriteRendererProperties]}
  {emitterName=Local_Sparkles, rendererClasses [&NiagaraSpriteRendererProperties]}
  {emitterName=Steam_Trail, rendererClasses [&NiagaraRibbonRendererProperties]}
  {emitterName=Trailing_Sparkles, rendererClasses [&NiagaraSpriteRendererProperties]}
 ]
 userVariables [
  User.Anim_LocalSpaceInheritance {type=NiagaraFloat, defaultValue=1, description=""}
  User.BaseColorAlpha {type=LinearColor, description="", defaultValue {struct=LinearColor, value=<rgba 1 1 1 0.5>}}
  User.EmissiveColor {type=LinearColor, description="", defaultValue {struct=LinearColor, value=<rgba 1 1 1 0.761>}}
  User.Global_Smoke_Opacity_Multiplier {type=NiagaraFloat, defaultValue=0, description=""}
  User.NormalizedNotifyProgress {type=NiagaraFloat, defaultValue=0, description=""}
  User.Opacity_Mult {type=NiagaraFloat, defaultValue=1, description=""}
  User.OwnerLocalVelocity {type=Vector3f, description="", defaultValue {struct=Vector3f, value=<0 0 0>}}
  User.RibbonScale {type=NiagaraFloat, defaultValue=1.5, description=""}
  User.Smoke_Alpha {type=NiagaraFloat, defaultValue=1, description=""}
  User.Smoke_Large_Offset {type=NiagaraFloat, defaultValue=0, description=""}
  User.Smoke_Position_Offset {type=NiagaraFloat, defaultValue=0, description=""}
  User.UseEmissiveSmoke {type=NiagaraBool}
 ]
}
